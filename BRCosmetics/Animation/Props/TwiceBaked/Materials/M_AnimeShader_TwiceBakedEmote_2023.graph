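# Cloth-shading anime material for the TwiceBaked emote prop, adding a transition that turns the surface to stone.
# It wraps the 2023 anime shader's Pre, Calculations and Post functions, blending stone lit and shaded colors, specular and an angle-corrected detail normal between them.
# A pre-skinned-position sphere mask, distorted by noise and driven by StoneProgress, forms the transition mask.
# CrackProgress scales an extra DFL texture added to the shader's DFL input.
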Material /BRCosmetics/Animation/Props/TwiceBaked/Materials/M_AnimeShader_TwiceBakedEmote_2023
graph {
 properties {
  bAutomaticallySetUsageInEditor=false
  bUseMaterialAttributes=true
  bUsedWithClothing=true
  bUsedWithMeshDeformer=true
  bUsedWithSkeletalMesh=true
  bUsedWithStaticLighting=true
  pixelDepthOffsetMode=PDOM_Legacy
  shadingModel=MSM_Cloth
  usedShadingModels=Cloth
 }
 nodes [
  n1 = MaterialFunctionCall!(CustomUV) {materialFunction=/Game/Global/Materials/Character_Parent_2023/Anime_Parent_2023/MaterialFunctions/MFC_AnimeShader_2023_Pre}
  n2 = NamedRerouteUsage()
  n3 = MaterialFunctionCall!(
   CustomLightVector
   OverrideNormal_WS
   STT=@n1.STT
   Normal_Map=LinearInterpolate(
    A=@n1.Normal
    B=MaterialFunctionCall(BaseNormal=@n1.Normal, AdditionalNormal=@n2) {materialFunction=&BlendAngleCorrectedNormals}.Result
    Alpha=NamedRerouteUsage()
   )
   DFL=Add(A=@n1.DFL, B=NamedRerouteUsage()) {constB=0}
   FX=@n1.STT
  ) {materialFunction=/Game/Global/Materials/Character_Parent_2023/Anime_Parent_2023/MaterialFunctions/MFC_AnimeShader_2023_Calculations}
  MaterialFunctionCall!(
   Shading_Gradient=@n3.Shading_Gradient
   Threshold_Gradient=@n3.Shading_Gradient
   Specular_Gradient=@n3.Shading_Gradient
   Sheen_Mask=@n3.Shading_Gradient
   Thickness=@n3.Shading_Gradient
   DFL_Mask=@n3.Shading_Gradient
   FX=@n3.Shading_Gradient
   Metallic_Gradient
   RimLight_Gradient=@n3.Shading_Gradient
   NormalBuffer=@n3.Shading_Gradient
   --------
   CINE_LightSources=@n3.Shading_Gradient
   CINE_LightColor=@n3.Shading_Gradient
   ---------
   Color_Lit=LinearInterpolate(A=@n1.Color_Lit, B=NamedRerouteUsage(), Alpha=NamedRerouteUsage())
   Color_Shaded=LinearInterpolate(A=@n1.Color_Shaded, B=NamedRerouteUsage(), Alpha=NamedRerouteUsage())
   SRM=@n1.SRM
   Emissive=@n1.SRM
   DFL_Color=@n1.SRM
   Specular_Color=LinearInterpolate(A=@n1.Specular_Color, B=ScalarParameter() {defaultValue=0.3, group=StoneTransit, parameterName=StoneSpec}, Alpha=NamedRerouteUsage()) {constB=0.5}
   RimLight_Color=@n1.SRM
  ) {materialFunction=/Game/Global/Materials/Character_Parent_2023/Anime_Parent_2023/MaterialFunctions/MFC_AnimeShader_2023_Post}
  n4 = VectorParameter() {group=StoneTransit, parameterName=StoneColorSimple, defaultValue=<rgba 0.1131 0.1249 0.1302 1>}
  NamedRerouteDeclaration(None=@n4.RGB) {name=StoneColorLit, nodeColor=<rgba 0 0.2969 1 1>}
  NamedRerouteDeclaration(None=Multiply(A=@n4.RGB, B=ScalarParameter() {defaultValue=0.3, group=StoneTransit, parameterName=StoneShadedMulti}) {constB=0.5}) {name=StoneColorShaded, nodeColor=<rgba 0.0312 0 1 1>}
  n5 = VertexInterpolator(VS=PreSkinnedPosition())
  n6 = ScalarParameter() {defaultValue=20, group=StoneTransit, parameterName=EffectRadius}
  n7 = SphereMask(A=@n5.PS, B=VectorParameter() {group=StoneTransit, parameterName=StartPosition, defaultValue=<rgba 0 0 0 1>}.RGB, Radius=@n6) {hardnessPercent=0}
  n8 = OneMinus(None=Saturate(None=@n7))
  n9 = TextureSampleParameter2D(UVs=Divide(A=ComponentMask(None=@n5.PS) {g=true, r=true}, B=@n6)) {
   group=StoneTransit
   parameterName=GradientDistort
   samplerType=SAMPLERTYPE_LinearColor
   texture=/Game/Global/Textures/Default/Noise/T_Noise01
  }
  n10 = Add(A=@n7, B=Multiply(A=Multiply(A=@n8, B=@n9.R), B=ScalarParameter() {defaultValue=0.3, group=StoneTransit, parameterName=DistortIntensity}))
  NamedRerouteDeclaration(None=SphereMask(
    A=Saturate(None=@n10)
    B=Constant() {r=1}
    Radius=ScalarParameter() {group=StoneTransit, parameterName=StoneProgress}
    Hardness=ScalarParameter() {defaultValue=90, group=StoneTransit, parameterName=StoneHardness}
   ) {hardnessPercent=90}) {name=Alpha, nodeColor=<rgba 1 0.4453 0 1>}
  n11 = TextureSampleParameter2D(UVs=Multiply(A=TextureCoordinate(), B=ScalarParameter() {defaultValue=3, group=StoneTransit, parameterName=NormalTilling})) {group=StoneTransit, parameterName=StoneNormal, samplerType=SAMPLERTYPE_Normal, texture=/Game/Global/Textures/Default/Noise/T_Noise_Normal_01}
  NamedRerouteDeclaration(None=@n11.RGB) {name=StoneNormal, nodeColor=<rgba 0 0.3906 1 1>}
  n12 = TextureSampleParameter2D() {
   group=StoneTransit
   parameterName=DFL_Ext
   samplerType=SAMPLERTYPE_Alpha
   sortPriority=14
   texture=/Game/Global/Textures/Default/Blanks/T_BlackAlpha
  }
  n13 = ScalarParameter() {parameterName=CrackProgress}
  NamedRerouteDeclaration(None=Multiply(A=Multiply(A=@n12.RGB, B=@n13), B=NamedRerouteUsage())) {name=Extra_DF, nodeColor=<rgba 1 0.9141 0 1>}
 ]
}
